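# Sound cue for entering the Zero Point bubble passthrough in the Cattus event, playing a single Snow_Enter_01 wave.
# It uses the Athena safe zone shield attenuation and the Athena ambient sound class, with boosted volume.
# A concurrency override allows one instance per owner and prevents new plays while one is active.
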SoundCue /FNGameplayCues/Athena/Sounds/Events/Cattus/ZeroPoint/ZeroPoint_Bubble_Passthrough_Enter_Cue
root=WavePlayer_0
nodes [{id=WavePlayer_0, properties {soundWaveAssetPtr=&/Snow_Enter_01}}]
properties {
 attenuationSettings=/Game/Sounds/Attenuation/Fort_SafeZone_Shield_Athena_Att
 bOverrideConcurrency=true
 maxDistance=4000
 soundClassObject=&Fort_Ambient_Athena
 volumeMultiplier=1.663
 concurrencyOverrides {bLimitToOwner=true, maxCount=1, resolutionRule=PreventNew}
}
