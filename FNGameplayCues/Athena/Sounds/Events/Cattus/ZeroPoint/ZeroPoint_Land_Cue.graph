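# Sound cue for a Zero Point landing in the Cattus event, built from three snow landing waves.
# A Random node picks one wave with equal weighting and feeds it through a Modulator left at default settings.
# It plays at reduced volume in the scripted events sound class, using status effect activation attenuation.
# It shares the LowGravityStone jump-land concurrency set to govern overlapping plays.
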SoundCue /FNGameplayCues/Athena/Sounds/Events/Cattus/ZeroPoint/ZeroPoint_Land_Cue
root=Modulator_0
nodes [
 {id=Modulator_0, children [Random_1]}
 {id=Random_1, children [WavePlayer_3,WavePlayer_4,WavePlayer_5], properties {weights [1,1,1]}}
 {id=WavePlayer_3, properties {soundWaveAssetPtr=&/Snow_Land_01}}
 {id=WavePlayer_4, properties {soundWaveAssetPtr=&/Snow_Land_02}}
 {id=WavePlayer_5, properties {soundWaveAssetPtr=&/Snow_Land_03}}
]
properties {
 attenuationSettings=/Game/Sounds/Attenuation/Fort_Ability_Status_Effect_Activation_Att
 maxDistance=6000
 soundClassObject=/Game/Sounds/FortSoundClassesAndModes/Fort_Events_Scripted
 volumeMultiplier=0.2632
 concurrencySet [/Game/Sounds/Fort_GamePlay_Sounds/Player/LowGravityStones/LowGravityStone_JumpLand_Concc]
}
